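; Gameplay ability for the FirePetal wandering boss's Smash bombardment: target-locks, takes over AI behavior and rains ground explosions.
; Chained EQS queries build shuffled, capped player-centered and enemy-centered location lists, plus, in the arena, lists for islands free of boss and target.
; A montage notify starts a timer spawning explosions from those lists with FX-manager idle effects.
; Critical weakpoints activate during the montage; ending resets FX, lock and control.
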(event Ability|EventActivateAbility
  (bind _targetcenteredtargetarray (Variables|InternalData|GetTargetCenteredTargetArray))
  (Development|PrintString "@@@ Event ActivateAbility [GA_Bombardment_FirePetal_WanderingBoss]: CALLED!" false (Variables|Debug|GetbDebugAbility))
  (Utilities|FlowControl|Sequence
    (:then_0
      (CallFunction|InitAbility)
      (bind _returnvalue (Ability|ApplyGameplayEffectToOwner (Variables|InternalData|GetGE_TargetLock)))
      (Variables|InternalData|SetTargetLockHandle _returnvalue)
      (bind _returnvalue_2 (Ability|CommitAbility))
      (if _returnvalue_2
        (bind _targetactor (Class|BPFLAICommonFunctions|GetTargetActorfromBlackboard (Variables|InternalData|GetFortPlayerPawn)))
        (bind _output_get (Variables|InternalData|SetCurrentTargetActor _targetactor))
        (Variables|InternalData|SetCurrentTargetLocation (AI|Navigation|ProjectPointtoNavigation (Transformation|GetActorLocation _output_get) 0 "/Game/Athena/AI/NavFilters/NavFilter_NPC_AnyTraversable.NavFilter_NPC_AnyTraversable_C"))
        ;[0.607 0.1344 1 0.5] Generate Player Centered Target List
        (bind _returnvalue_3 (AI|EQS|RunEQSQuery "/FirePetal_WanderingBoss/AI/EQS/EQS_Bombardment_PlayerCenter_FirePetal_WanderingBoss.EQS_Bombardment_PlayerCenter_FirePetal_WanderingBoss" (Variables|InternalData|GetFortPlayerPawn) "AllMatching"))
        (Utilities|IsValid _returnvalue_3
          (:Is%20Valid
            (AI|EQS|BindEventtoOnQueryFinished _returnvalue_3 (delegate Custom|OnTargetCenteredQueryFinished)))
          (:Is%20Not%20Valid
            (Development|PrintString "@@@@@@ Event ActivateAbility ERROR - Player Query Instance INVALID!" false (Variables|Debug|GetbDebugAbility))
            (Utilities|Array|AddUnique _targetcenteredtargetarray (Variables|InternalData|GetCurrentTargetLocation))
            (Variables|/FirePetal_WanderingBoss/NPCs/WanderingBoss/Pawn/BP_PlayerPawn_FirePetal_WanderingBoss.BP_PlayerPawn_FirePetal_WanderingBoss_C|SetBombardment_TargetCenteredTargetArray _targetcenteredtargetarray (Variables|InternalData|GetWanderingBossPawn))))
        (else
          (Ability|EndAbility))))
    (:then_1
      (bind _returnvalue_4 (Building|SpawnBuildingGameplayActor "/FirePetal_WanderingBoss/GameplayActors/BGA_FirePetal_WanderingBoss_BombardmentFXManager_Smash.BGA_FirePetal_WanderingBoss_BombardmentFXManager_Smash_C" (Math|Transform|MakeTransform (Transformation|GetActorLocation (Variables|InternalData|GetFortPlayerPawn))) (Variables|InternalData|GetFortPlayerPawn) true false))
      (bind _asbga_fire_petal_wandering_boss_bombardment_fxmanager (Utilities|Casting|CastToBGA_FirePetal_WanderingBoss_BombardmentFXManager _returnvalue_4))
      (Variables|InternalData|SetBGA_FXManager _asbga_fire_petal_wandering_boss_bombardment_fxmanager))))

(event Custom|OnTargetCenteredQueryFinished
  (params (QueryInstance :type EQS%20Query%20Instance%20Object%20Reference) (QueryStatus :type EEnvQueryStatus%20Enum :default "Processing"))
  (bind _sf_initialtargetcount_playercenetered (Variables|Default|GetSF_InitialTargetCount_PlayerCenetered))
  (bind _returnvalue (ScalableFloat|GetValueatLevel _sf_initialtargetcount_playercenetered))
  (bind _targetcenteredtargetarray (Variables|InternalData|GetTargetCenteredTargetArray))
  ;[0.607 0.1344 1 0.5] Generate Player Centered Target List
  (if (Utilities|Enum|Equal(Enum) QueryStatus "Success")
    (bind (_resultlocations _returnvalue_2) (AI|EQS|GetQueryResultsasLocations QueryInstance))
    (if _returnvalue_2
      (bind _output_get (Variables|InternalData|SetTargetCenteredTargetArray _resultlocations))
      (Utilities|Array|Shuffle _output_get)
      (else
        (Development|PrintString "@@@@@@ Event ActivateAbility ERROR - NO valid reults for Player!" false (Variables|Debug|GetbDebugAbility))))
    (else
      (Development|PrintString "@@@@@@ Event ActivateAbility ERROR -Player Query Status NOT Success!" false (Variables|Debug|GetbDebugAbility))))
  (Utilities|Array|Insert _targetcenteredtargetarray (Variables|InternalData|GetCurrentTargetLocation))
  (if (> (Utilities|Array|Length _targetcenteredtargetarray) _returnvalue)
    (Utilities|Array|Resize (Variables|InternalData|GetTargetCenteredTargetArray) (Math|Float|Truncate _returnvalue)))
  (Variables|/FirePetal_WanderingBoss/NPCs/WanderingBoss/Pawn/BP_PlayerPawn_FirePetal_WanderingBoss.BP_PlayerPawn_FirePetal_WanderingBoss_C|SetBombardment_TargetCenteredTargetArray (Variables|InternalData|GetTargetCenteredTargetArray) (Variables|InternalData|GetWanderingBossPawn))
  (if (Variables|Debug|GetbDebugAbility)
    (CallFunction|DebugPrintArray (Variables|InternalData|GetTargetCenteredTargetArray) "(R=0.916948,G=0.000000,B=1.000000,A=1.000000)"))
  (CallFunction|GenerateEnemyCenteredTargets))

(event Custom|GenerateEnemyCenteredTargets
  ;[0.901 0.2 0.02851 0.5] Generate Enemy Centered Target List
  (bind _returnvalue (AI|EQS|RunEQSQuery "/FirePetal_WanderingBoss/AI/EQS/EQS_Bombardment_EnemyCenter_FirePetal_WanderingBoss.EQS_Bombardment_EnemyCenter_FirePetal_WanderingBoss" (Variables|InternalData|GetFortPlayerPawn) "AllMatching"))
  (Utilities|IsValid _returnvalue
    (:Is%20Valid
      (AI|EQS|BindEventtoOnQueryFinished _returnvalue (delegate Custom|OnEnemyCenteredQueryFinished)))
    (:Is%20Not%20Valid
      (Development|PrintString "@@@@@@ Event ActivateAbility ERROR - Enemy Query Instance INVALID!" false (Variables|Debug|GetbDebugAbility)))))

(event Custom|OnEnemyCenteredQueryFinished
  (params (QueryInstance :type EQS%20Query%20Instance%20Object%20Reference) (QueryStatus :type EEnvQueryStatus%20Enum :default "Processing"))
  (bind _sf_initialtargetcount_enemycenetered (Variables|Default|GetSF_InitialTargetCount_EnemyCenetered))
  (bind _returnvalue (ScalableFloat|GetValueatLevel _sf_initialtargetcount_enemycenetered))
  ;[0.901 0.2 0.02851 0.5] Generate Enemy Centered Target List
  (if (Utilities|Enum|Equal(Enum) QueryStatus "Success")
    (bind (_resultlocations _returnvalue_2) (AI|EQS|GetQueryResultsasLocations QueryInstance))
    (if _returnvalue_2
      (bind _output_get (Variables|InternalData|SetEnemyCenteredTargetArray _resultlocations))
      (Utilities|Array|Shuffle _output_get)
      (else
        (Development|PrintString "@@@@@@ Event ActivateAbility ERROR - NO valid reults for Enemy!" false (Variables|Debug|GetbDebugAbility))))
    (else
      (Development|PrintString "@@@@@@ Event ActivateAbility ERROR -Enemy Query Status NOT Success!" false (Variables|Debug|GetbDebugAbility))))
  (if (> (Utilities|Array|Length (Variables|InternalData|GetEnemyCenteredTargetArray)) _returnvalue)
    (Utilities|Array|Resize (Variables|InternalData|GetEnemyCenteredTargetArray) (Math|Float|Truncate _returnvalue)))
  (Variables|/FirePetal_WanderingBoss/NPCs/WanderingBoss/Pawn/BP_PlayerPawn_FirePetal_WanderingBoss.BP_PlayerPawn_FirePetal_WanderingBoss_C|SetBombardment_EnemyCenteredTargetArray (Variables|InternalData|GetEnemyCenteredTargetArray) (Variables|InternalData|GetWanderingBossPawn))
  (if (Variables|Debug|GetbDebugAbility)
    (CallFunction|DebugPrintArray (Variables|InternalData|GetEnemyCenteredTargetArray) "(R=0.916948,G=0.000000,B=1.000000,A=1.000000)"))
  (if (Variables|InternalData|GetbIsInArena)
    (CallFunction|GenerateNeutralIslandCenteredTargets)
    (else
      (CallFunction|InitiateBombardment))))

(event Custom|GenerateNeutralIslandCenteredTargets
  ;[0.03522 0.6875 0.5914 0.5] Generate Arena Island Target Lists
  (Development|PrintString "@@@@@@ GenerateNeutralIslandCenteredTargets [GA_Bombardment_FirePetal_WanderingBoss]: CALLED!" false (Variables|Debug|GetbDebugAbility))
  (Variables|Arena|SetSensorIndex 0)
  (Variables|InternalData|SetArena%20Island%20Sensor%20Array (Variables|/FirePetal_WanderingBoss/NPCs/WanderingBoss/Pawn/BP_PlayerPawn_FirePetal_WanderingBoss.BP_PlayerPawn_FirePetal_WanderingBoss_C|GetArenaIslandSensorArray (Variables|InternalData|GetWanderingBossPawn)))
  (bind _currentsensor (CallFunction|GetCurrentSensor))
  (Variables|InternalData|SetSensorIterator _currentsensor)
  (Utilities|Array|Clear (Variables|InternalData|GetIslandCenteredTargetArray))
  (if (not (or (Collision|IsOverlappingActor (Variables|InternalData|GetSensorIterator) (Variables|InternalData|GetCurrentTargetActor) :!class /Script/Engine.Actor) (Collision|IsOverlappingActor (Variables|InternalData|GetSensorIterator) (Variables|InternalData|GetWanderingBossPawn) :!class /Script/Engine.Actor)))
    (Development|PrintString (Utilities|String|Append "@@@@@@@@@ GenerateNeutralIslandCenteredTargets [GA_Bombardment_FirePetal_WanderingBoss]: [" (Utilities|String|ToString(Integer) (Variables|Arena|GetSensorIndex)) "] " (Utilities|GetDisplayName (Variables|InternalData|GetSensorIterator)) "= NEUTRAL! ") false (Variables|Debug|GetbDebugAbility))
    (Variables|/FirePetal_WanderingBoss/NPCs/WanderingBoss/Pawn/BP_PlayerPawn_FirePetal_WanderingBoss.BP_PlayerPawn_FirePetal_WanderingBoss_C|SetSensorArrayIndex (Variables|Arena|GetSensorIndex) (Variables|InternalData|GetWanderingBossPawn))
    (bind _returnvalue (AI|EQS|RunEQSQuery "/FirePetal_WanderingBoss/AI/EQS/EQS_Bombardment_IslandCenter_FirePetal_WanderingBoss.EQS_Bombardment_IslandCenter_FirePetal_WanderingBoss" (Variables|InternalData|GetFortPlayerPawn) "AllMatching"))
    (Utilities|IsValid _returnvalue
      (:Is%20Valid
        (AI|EQS|BindEventtoOnQueryFinished _returnvalue (delegate Custom|OnIslandCenteredQueryFinished)))
      (:Is%20Not%20Valid
        (Development|PrintString "@@@@@@ Event ActivateAbility ERROR - Island Query Instance INVALID!" false (Variables|Debug|GetbDebugAbility))))
    (else
      (Development|PrintString (Utilities|String|Append "@@@@@@@@@ GenerateNeutralIslandCenteredTargets [GA_Bombardment_FirePetal_WanderingBoss]: [" (Utilities|String|ToString(Integer) (Variables|Arena|GetSensorIndex)) "] " (Utilities|GetDisplayName (Variables|InternalData|GetSensorIterator)) "= OCCUPIED! ") false (Variables|Debug|GetbDebugAbility))
      (Utilities|Array|SetArrayElem (Variables|Arena|GetIslandTargetArrayStruct) (Variables|Arena|GetSensorIndex) (Utilities|Struct|MakeBombardmentIslandTargetArrays (Variables|InternalData|GetIslandCenteredTargetArray)))
      (if (Variables|Debug|GetbDebugAbility)
        (CallFunction|DebugPrintArray (Variables|InternalData|GetIslandCenteredTargetArray) "(R=0.916948,G=0.000000,B=1.000000,A=1.000000)"))
      (bind _result (Math|Integer|IncrementInt (Variables|Arena|GetSensorIndex))
        (:%20%20
          (if (< _result (Utilities|Array|Length (Variables|/FirePetal_WanderingBoss/NPCs/WanderingBoss/Pawn/BP_PlayerPawn_FirePetal_WanderingBoss.BP_PlayerPawn_FirePetal_WanderingBoss_C|GetArenaIslandSensorArray (Variables|InternalData|GetWanderingBossPawn))))
            (CallFunction|RefreshIslandTargets)
            (else
              (CallFunction|InitiateBombardment))))))))

(event Custom|RefreshIslandTargets
  ;[0.03522 0.6875 0.5914 0.5] Generate Arena Island Target Lists
  (bind _currentsensor (CallFunction|GetCurrentSensor))
  (Variables|InternalData|SetSensorIterator _currentsensor)
  (Utilities|Array|Clear (Variables|InternalData|GetIslandCenteredTargetArray))
  (if (not (or (Collision|IsOverlappingActor (Variables|InternalData|GetSensorIterator) (Variables|InternalData|GetCurrentTargetActor) :!class /Script/Engine.Actor) (Collision|IsOverlappingActor (Variables|InternalData|GetSensorIterator) (Variables|InternalData|GetWanderingBossPawn) :!class /Script/Engine.Actor)))
    (Development|PrintString (Utilities|String|Append "@@@@@@@@@ GenerateNeutralIslandCenteredTargets [GA_Bombardment_FirePetal_WanderingBoss]: [" (Utilities|String|ToString(Integer) (Variables|Arena|GetSensorIndex)) "] " (Utilities|GetDisplayName (Variables|InternalData|GetSensorIterator)) "= NEUTRAL! ") false (Variables|Debug|GetbDebugAbility))
    (Variables|/FirePetal_WanderingBoss/NPCs/WanderingBoss/Pawn/BP_PlayerPawn_FirePetal_WanderingBoss.BP_PlayerPawn_FirePetal_WanderingBoss_C|SetSensorArrayIndex (Variables|Arena|GetSensorIndex) (Variables|InternalData|GetWanderingBossPawn))
    (bind _returnvalue (AI|EQS|RunEQSQuery "/FirePetal_WanderingBoss/AI/EQS/EQS_Bombardment_IslandCenter_FirePetal_WanderingBoss.EQS_Bombardment_IslandCenter_FirePetal_WanderingBoss" (Variables|InternalData|GetFortPlayerPawn) "AllMatching"))
    (Utilities|IsValid _returnvalue
      (:Is%20Valid
        (AI|EQS|BindEventtoOnQueryFinished _returnvalue (delegate Custom|OnIslandCenteredQueryFinished)))
      (:Is%20Not%20Valid
        (Development|PrintString "@@@@@@ Event ActivateAbility ERROR - Island Query Instance INVALID!" false (Variables|Debug|GetbDebugAbility))))
    (else
      (Development|PrintString (Utilities|String|Append "@@@@@@@@@ GenerateNeutralIslandCenteredTargets [GA_Bombardment_FirePetal_WanderingBoss]: [" (Utilities|String|ToString(Integer) (Variables|Arena|GetSensorIndex)) "] " (Utilities|GetDisplayName (Variables|InternalData|GetSensorIterator)) "= OCCUPIED! ") false (Variables|Debug|GetbDebugAbility))
      (Utilities|Array|SetArrayElem (Variables|Arena|GetIslandTargetArrayStruct) (Variables|Arena|GetSensorIndex) (Utilities|Struct|MakeBombardmentIslandTargetArrays (Variables|InternalData|GetIslandCenteredTargetArray)))
      (if (Variables|Debug|GetbDebugAbility)
        (CallFunction|DebugPrintArray (Variables|InternalData|GetIslandCenteredTargetArray) "(R=0.916948,G=0.000000,B=1.000000,A=1.000000)"))
      (bind _result (Math|Integer|IncrementInt (Variables|Arena|GetSensorIndex))
        (:%20%20
          (if (< _result (Utilities|Array|Length (Variables|/FirePetal_WanderingBoss/NPCs/WanderingBoss/Pawn/BP_PlayerPawn_FirePetal_WanderingBoss.BP_PlayerPawn_FirePetal_WanderingBoss_C|GetArenaIslandSensorArray (Variables|InternalData|GetWanderingBossPawn))))
            (CallFunction|RefreshIslandTargets)
            (else
              (CallFunction|InitiateBombardment))))))))

(event Custom|OnIslandCenteredQueryFinished
  (params (QueryInstance :type EQS%20Query%20Instance%20Object%20Reference) (QueryStatus :type EEnvQueryStatus%20Enum :default "Processing"))
  (bind _sf_arenatargetcount_neutralisland (Variables|Default|GetSF_ArenaTargetCount_NeutralIsland))
  (bind _returnvalue (ScalableFloat|GetValueatLevel _sf_arenatargetcount_neutralisland))
  ;[0.03522 0.6875 0.5914 0.5] Generate Arena Island Target Lists
  (if (Utilities|Enum|Equal(Enum) QueryStatus "Success")
    (bind (_resultlocations _returnvalue_2) (AI|EQS|GetQueryResultsasLocations QueryInstance))
    (if _returnvalue_2
      (bind _output_get (Variables|InternalData|SetIslandCenteredTargetArray _resultlocations))
      (Utilities|Array|Shuffle _output_get)
      (else
        (Development|PrintString "@@@@@@ Event ActivateAbility ERROR - NO valid reults for Island!" false (Variables|Debug|GetbDebugAbility))))
    (else
      (Development|PrintString "@@@@@@ Event ActivateAbility ERROR -Island Query Status NOT Success!" false (Variables|Debug|GetbDebugAbility))))
  (if (> (Utilities|Array|Length (Variables|InternalData|GetIslandCenteredTargetArray)) _returnvalue)
    (Utilities|Array|Resize (Variables|InternalData|GetIslandCenteredTargetArray) (Math|Float|Truncate _returnvalue)))
  (Utilities|Array|SetArrayElem (Variables|Arena|GetIslandTargetArrayStruct) (Variables|Arena|GetSensorIndex) (Utilities|Struct|MakeBombardmentIslandTargetArrays (Variables|InternalData|GetIslandCenteredTargetArray)))
  (if (Variables|Debug|GetbDebugAbility)
    (CallFunction|DebugPrintArray (Variables|InternalData|GetIslandCenteredTargetArray) "(R=0.916948,G=0.000000,B=1.000000,A=1.000000)"))
  (bind _result (Math|Integer|IncrementInt (Variables|Arena|GetSensorIndex))
    (:%20%20
      (if (< _result (Utilities|Array|Length (Variables|/FirePetal_WanderingBoss/NPCs/WanderingBoss/Pawn/BP_PlayerPawn_FirePetal_WanderingBoss.BP_PlayerPawn_FirePetal_WanderingBoss_C|GetArenaIslandSensorArray (Variables|InternalData|GetWanderingBossPawn))))
        (CallFunction|RefreshIslandTargets)
        (else
          (CallFunction|InitiateBombardment))))))

(event Custom|InitiateBombardment
  (bind _sf_updatehealthwidgetoffset (Variables|Default|GetSF_UpdateHealthWidgetOffset))
  (bind _returnvalue (ScalableFloat|GetValueasBool _sf_updatehealthwidgetoffset))
  (bind _montageinfo_montagetoplay (Utilities|Struct|BreakFortGameplayAbilityMontageInfo (Variables|GetMontageInfo_MontageToPlay)))
  (bind _wanderingbosspawn (Variables|InternalData|GetWanderingBossPawn))
  (bind (_asynctaskproxy _notifytag) (Ability|Tasks|PlayMontageAndWaitNotify "WanderingBoss_Bombardment" _montageinfo_montagetoplay)
    (:then
      (CallFunction|UpdateWeakpoints true))
    (:OnCompleted
      _)
    (:OnInterrupted
      _)
    (:OnCancelled
      _)
    (:OnNotifyBegin
      (switch Utilities|FlowControl|Switch|SwitchonGameplayTag _notifytag
        (:Event.WanderingBoss.Bombardment.Fire
          (Variables|InternalData|SetExplosionCounter 0)
          (bind _returnvalue_2 (Utilities|Time|SetTimerbyEvent (delegate Custom|SpawnNextExplosions) (ScalableFloat|GetValueatLevel (Variables|Default|GetSF_AOESpawnDelay)) true))
          (Variables|InternalData|SetBombardmentTimerHandle _returnvalue_2)
          (if _returnvalue
            (Class|BPPlayerPawnFirePetalWanderingBoss|StartWidgetVerticalOffset _wanderingbosspawn -1.2e+02 1.75)))
        (:Event.WanderingBoss.Bombardment.End
          (CallFunction|UpdateWeakpoints false)
          (if _returnvalue
            (Class|BPPlayerPawnFirePetalWanderingBoss|ResetWidgetVerticalOffset _wanderingbosspawn 1.5))))))
  (Ability|EndAbility))

(event Custom|SpawnNextExplosions
  (bind _explosioncounter (Variables|InternalData|GetExplosionCounter))
  (bind _targetcenteredtargetarray (Variables|InternalData|GetTargetCenteredTargetArray))
  (bind _output (Utilities|Array|Get(acopy) _targetcenteredtargetarray _explosioncounter))
  (bind _explosioncounter_2 (Variables|InternalData|GetExplosionCounter))
  (bind _enemycenteredtargetarray (Variables|InternalData|GetEnemyCenteredTargetArray))
  (bind _output_2 (Utilities|Array|Get(acopy) _enemycenteredtargetarray _explosioncounter_2))
  (bind _returnvalue (Building|SpawnBuildingGameplayActor "/FirePetal_WanderingBoss/GameplayActors/BGA_FirePetal_WanderingBoss_GroundExplosion.BGA_FirePetal_WanderingBoss_GroundExplosion_C" (Math|Transform|MakeTransform _output) (Variables|InternalData|GetFortPlayerPawn) true false))
  (bind _asbga_fire_petal_wandering_boss_ground_explosion (Utilities|Casting|CastToBGA_FirePetal_WanderingBoss_GroundExplosion _returnvalue)
    (:then
      (Class|BGAFirePetalWanderingBossBombardmentFXManager|CreateANewIdleFX (Variables|InternalData|GetBGA_FXManager) (+ _output "0.000000,0.000000,-20.000000") _asbga_fire_petal_wandering_boss_ground_explosion)
      (Utilities|Array|AddUnique (Variables|/FirePetal_WanderingBoss/GameplayActors/BGA_FirePetal_WanderingBoss_GroundExplosion.BGA_FirePetal_WanderingBoss_GroundExplosion_C|GetActors%20to%20Ignore _asbga_fire_petal_wandering_boss_ground_explosion) (Variables|InternalData|GetFortPlayerPawn)))
    (:CastFailed
      _))
  (bind _returnvalue_2 (Building|SpawnBuildingGameplayActor "/FirePetal_WanderingBoss/GameplayActors/BGA_FirePetal_WanderingBoss_GroundExplosion.BGA_FirePetal_WanderingBoss_GroundExplosion_C" (Math|Transform|MakeTransform _output_2) (Variables|InternalData|GetFortPlayerPawn) true false))
  (bind _asbga_fire_petal_wandering_boss_ground_explosion_2 (Utilities|Casting|CastToBGA_FirePetal_WanderingBoss_GroundExplosion _returnvalue_2)
    (:then
      (Class|BGAFirePetalWanderingBossBombardmentFXManager|CreateANewIdleFX (Variables|InternalData|GetBGA_FXManager) (+ _output_2 "0.000000,0.000000,-20.000000") _asbga_fire_petal_wandering_boss_ground_explosion_2)
      (Utilities|Array|AddUnique (Variables|/FirePetal_WanderingBoss/GameplayActors/BGA_FirePetal_WanderingBoss_GroundExplosion.BGA_FirePetal_WanderingBoss_GroundExplosion_C|GetActors%20to%20Ignore _asbga_fire_petal_wandering_boss_ground_explosion_2) (Variables|InternalData|GetFortPlayerPawn)))
    (:CastFailed
      _))
  (if (Variables|InternalData|GetbIsInArena)
    (Utilities|FlowControl|Delay (ScalableFloat|GetValueatLevel (Variables|Default|GetSF_NeutralIsland_AOESpawnDelay)))
    (CallFunction|ProcessArenaExplosions (Variables|InternalData|GetExplosionCounter)))
  (Math|Integer|IncrementInt (Variables|InternalData|GetExplosionCounter)
    (:%20%20
      (if (>= (Variables|InternalData|GetExplosionCounter) (ScalableFloat|GetValueasInteger (Variables|Default|GetSF_InitialTargetCount_PlayerCenetered)))
        (Utilities|Time|ClearandInvalidateTimerbyHandle (Variables|InternalData|GetBombardmentTimerHandle))))))

(event Ability|EventOnEndAbility
  (params (bWasCancelled :type Boolean))
  (bind _aibotcontroller (Variables|InternalData|GetAIBotController))
  (bind _wanderingbosspawn (Variables|InternalData|GetWanderingBossPawn))
  (Development|PrintString "@@@ Event OnEndAbility [GA_Bombardment_FirePetal_WanderingBoss]: CALLED!" false (Variables|Debug|GetbDebugAbility))
  (Class|BGAFirePetalWanderingBossBombardmentFXManager|Reset (Variables|InternalData|GetBGA_FXManager))
  (Ability|RemoveGameplayEffectFromOwnerWithHandle (Variables|InternalData|GetTargetLockHandle) 1)
  (if (Variables|/Game/Athena/AI/NPCs/Base/BP_PlayerPawn_NonParticipant_Boss.BP_PlayerPawn_NonParticipant_Boss_C|GetIsWidgetOffsetOverriden _wanderingbosspawn)
    (Variables|/Game/Athena/AI/NPCs/Base/BP_PlayerPawn_NonParticipant_Boss.BP_PlayerPawn_NonParticipant_Boss_C|SetIsWidgetOffsetOverriden false _wanderingbosspawn))
  (if (Variables|/FirePetal_WanderingBoss/NPCs/WanderingBoss/Pawn/BP_PlayerPawn_FirePetal_WanderingBoss.BP_PlayerPawn_FirePetal_WanderingBoss_C|GetbWeakpointsActive (Variables|InternalData|GetWanderingBossPawn))
    (CallFunction|UpdateWeakpoints false))
  (Utilities|IsValid _aibotcontroller
    (:Is%20Valid
      (AI|PopBehaviorControl _aibotcontroller))))

(fn DebugPrintArray
  (params (LocationArray :type Array%20of%20Vectors) (LocationColor :type Linear%20Color%20Structure))
  (Development|PrintString (Utilities|String|Append "@@@@@@@@@ DebugPrintArray [GA_Bombardment_FirePetal_WanderingBoss]: CALLED with a Location Array of size " (Utilities|String|ToString(Integer) (Utilities|Array|Length (Variables|GetLocationArray))) "!") false (Variables|Debug|GetbDebugAbility))
  (for _array_element (Variables|GetLocationArray)
    (Rendering|Debug|DrawDebugSphere _array_element 121.0 12 (Variables|GetLocationColor) 1e+01 2.0)))

(fn GetCurrentSensor
  (return (Utilities|Array|Get(acopy) (Variables|/FirePetal_WanderingBoss/NPCs/WanderingBoss/Pawn/BP_PlayerPawn_FirePetal_WanderingBoss.BP_PlayerPawn_FirePetal_WanderingBoss_C|GetArenaIslandSensorArray (Variables|InternalData|GetWanderingBossPawn)) (Variables|Arena|GetSensorIndex))))

(fn InitAbility
  (bind _output_get (Variables|InternalData|SetFortPlayerPawn (CallFunction|GetAvatarasFortPlayerPawn)))
  (bind _returnvalue (Class|FortVehicleInterface|GetSkeletalMeshComponent(Message) _output_get))
  (Variables|Default|SetSkelMeshComponent _returnvalue)
  (bind _asfort_athena_aibot_controller (Utilities|Casting|CastToFortAthenaAIBotController (Pawn|GetController (Variables|InternalData|GetFortPlayerPawn)))
    (:then
      (bind _output_get_2 (Variables|InternalData|SetAIBotController _asfort_athena_aibot_controller))
      (AI|PushBehaviorControl _output_get_2 -2097152)
      (Class|FortAthenaAIBotController|StopFire (Variables|InternalData|GetAIBotController))
      (Utilities|Array|Clear (Variables|InternalData|GetTargetCenteredTargetArray))
      (Utilities|Array|Clear (Variables|InternalData|GetEnemyCenteredTargetArray))
      (for _array_element (Variables|Arena|GetIslandTargetArrayStruct)
        (bind _array_element_islandtargetarray_12_0e8f83d44907c173f67d088eac9318b2 (Utilities|Struct|BreakBombardmentIslandTargetArrays _array_element))
        (Utilities|Array|Clear _array_element_islandtargetarray_12_0e8f83d44907c173f67d088eac9318b2))
      (bind _asbp_player_pawn_fire_petal_wandering_boss (Utilities|Casting|CastToBP_PlayerPawn_FirePetal_WanderingBoss (Variables|InternalData|GetFortPlayerPawn))
        (:then
          (bind _output_get_3 (Variables|InternalData|SetWanderingBossPawn _asbp_player_pawn_fire_petal_wandering_boss))
          (Utilities|Array|Clear (Variables|/FirePetal_WanderingBoss/NPCs/WanderingBoss/Pawn/BP_PlayerPawn_FirePetal_WanderingBoss.BP_PlayerPawn_FirePetal_WanderingBoss_C|GetBombardment_TargetCenteredTargetArray _output_get_3))
          (Utilities|Array|Clear (Variables|/FirePetal_WanderingBoss/NPCs/WanderingBoss/Pawn/BP_PlayerPawn_FirePetal_WanderingBoss.BP_PlayerPawn_FirePetal_WanderingBoss_C|GetBombardment_EnemyCenteredTargetArray (Variables|InternalData|GetWanderingBossPawn)))
          (bind _bdidactorhavetags (Class|AthenaQuestFunctionLibrary|AthenaCheckActorHasTags (Variables|InternalData|GetWanderingBossPawn) "(GameplayTags=((TagName=\"FirePetalBosses.Demons.IsInArena\")))" true))
          (Variables|InternalData|SetbIsInArena _bdidactorhavetags))
        (:CastFailed
          (Ability|EndAbility))))
    (:CastFailed
      (Ability|EndAbility))))

(fn ProcessArenaExplosions
  (params (ExplosionIndex :type Integer))
  (bind _loctargetarrayiterator (LocalVariables|GetlocTargetArrayIterator))
  (bind _islandtargetarraystruct (Variables|Arena|GetIslandTargetArrayStruct))
  (bind _output (Utilities|Array|Get(acopy) _islandtargetarraystruct _loctargetarrayiterator))
  (bind _islandtargetarray_12_0e8f83d44907c173f67d088eac9318b2 (Utilities|Struct|BreakBombardmentIslandTargetArrays _output))
  (bind _explosioncounter (Variables|InternalData|GetExplosionCounter))
  (bind _output_2 (Utilities|Array|Get(acopy) _islandtargetarray_12_0e8f83d44907c173f67d088eac9318b2 _explosioncounter))
  (Utilities|FlowControl|Sequence
    (:then_0
      (LocalVariables|SetlocTargetArrayIterator 0))
    (:then_1
      (LocalVariables|SetlocTargetArrayIterator 1))
    (:then_2
      (LocalVariables|SetlocTargetArrayIterator 2))
    (:then_3
      (LocalVariables|SetlocTargetArrayIterator 3))
    (:then_4
      (LocalVariables|SetlocTargetArrayIterator 4)))
  (if (and (> (Utilities|Array|Length _islandtargetarray_12_0e8f83d44907c173f67d088eac9318b2) 0) (< (Variables|InternalData|GetExplosionCounter) (Utilities|Array|Length _islandtargetarray_12_0e8f83d44907c173f67d088eac9318b2)))
    (bind _returnvalue (Building|SpawnBuildingGameplayActor "/FirePetal_WanderingBoss/GameplayActors/BGA_FirePetal_WanderingBoss_GroundExplosion.BGA_FirePetal_WanderingBoss_GroundExplosion_C" (Math|Transform|MakeTransform _output_2) (Variables|InternalData|GetFortPlayerPawn) true false))
    (bind _asbga_fire_petal_wandering_boss_ground_explosion (Utilities|Casting|CastToBGA_FirePetal_WanderingBoss_GroundExplosion _returnvalue))
    (Class|BGAFirePetalWanderingBossBombardmentFXManager|CreateANewIdleFX (Variables|InternalData|GetBGA_FXManager) (+ _output_2 "0.000000,0.000000,-20.000000") _asbga_fire_petal_wandering_boss_ground_explosion)
    (Utilities|Array|AddUnique (Variables|/FirePetal_WanderingBoss/GameplayActors/BGA_FirePetal_WanderingBoss_GroundExplosion.BGA_FirePetal_WanderingBoss_GroundExplosion_C|GetActors%20to%20Ignore _asbga_fire_petal_wandering_boss_ground_explosion) (Variables|InternalData|GetFortPlayerPawn))))

(fn UpdateWeakpoints
  (params (bWeakpointsActive :type Boolean))
  (bind _wanderingbosspawn (Variables|InternalData|GetWanderingBossPawn))
  (bind _wanderingbosspawn_2 (Variables|InternalData|GetWanderingBossPawn))
  (if bWeakpointsActive
    (Variables|/FirePetal_WanderingBoss/NPCs/WanderingBoss/Pawn/BP_PlayerPawn_FirePetal_WanderingBoss.BP_PlayerPawn_FirePetal_WanderingBoss_C|SetbWeakpointsActive true _wanderingbosspawn_2)
    (DamageZones|UpdateBonesatDamageZoneType _wanderingbosspawn_2 "Critical" (Variables|InternalData|GetWeakpointBones))
    (bind _returnvalue (Ability|ApplyGameplayEffectToOwner (Variables|InternalData|GetGE_BombardmentWeakpointTracker)))
    (Variables|InternalData|SetWeakpointTrackerHandle _returnvalue)
    (else
      (Variables|/FirePetal_WanderingBoss/NPCs/WanderingBoss/Pawn/BP_PlayerPawn_FirePetal_WanderingBoss.BP_PlayerPawn_FirePetal_WanderingBoss_C|SetbWeakpointsActive false _wanderingbosspawn)
      (DamageZones|UpdateBonesatDamageZoneType _wanderingbosspawn "Critical" (Variables|InternalData|GetWeakpointDefaultBones))
      (Ability|RemoveGameplayEffectFromOwnerWithHandle (Variables|InternalData|GetWeakpointTrackerHandle)))))
