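; Spawn-in gameplay ability for the FirePetal wandering boss (Smash) that reveals the boss and knocks back whatever is targeted around it.
; It plays the spawn montage, unhides the AI pawn after a tiny delay to avoid animation pops, and ends when the montage finishes.
; Alongside this, a briefly delayed target selection pushes pawns and vehicles away from the boss using separate scalable forward/height velocities, then applies a gameplay effect container.
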(event Ability|EventActivateAbility
  (bind (_montageinfo_montagetoplay _montageinfo_animplayrate _montageinfo_animrootmotiontranslationscale) (Utilities|Struct|BreakFortGameplayAbilityMontageInfo (Variables|GetMontageInfo_MontageToPlay)))
  (|CallParentFunction)
  (Development|PrintString "@@@ Event ActivateAbility [GA_SpawnIn_FirePetal_WanderingBoss]: CALLED!" false (Variables|Debug|GetbDebugAbility))
  (Utilities|FlowControl|Sequence
    (:then_0
      (bind _asynctaskproxy (Ability|Tasks|PlayMontageAndWaitNotify "WanderingBoss_Spawn" _montageinfo_montagetoplay _montageinfo_animplayrate "None" true _montageinfo_animrootmotiontranslationscale)
        (:then
          (Variables|InternalData|SetSpawnMontageTask _asynctaskproxy)
          (Utilities|FlowControl|Delay 0.0001) ; Tiny delay otherwise the ability anims may pop
          (Rendering|SetActorHiddenInGame (Variables|InternalData|GetAIPawn) false))
        (:OnCompleted
          _)
        (:OnInterrupted
          _)
        (:OnCancelled
          _))
      (Ability|EndAbility))
    (:then_1
      (Utilities|FlowControl|Delay 0.1)
      (Development|PrintString "@@@@@@ Event ActivateAbility [GA_SpawnIn_FirePetal_WanderingBoss]: Knockback effect triggered!" false (Variables|Debug|GetbDebugAbility))
      (bind (_asynctaskproxy_2 _targetdata _applicationtag) (Ability|Tasks|DoTargetSelection "WanderingBoss_Spawn_Knockback" "(TagName=\"Targeting.ApplicationTag.0\")" true)
        (:Targeted
          (for _array_element (Ability|TargetData|GetAllActorsfromTargetData _targetdata)
            (bind _asfort_pawn (Utilities|Casting|CastToFortPawn _array_element)
              (:then
                (Default|KnockbackPawn _asfort_pawn))
              (:CastFailed
                (bind _asfort_athena_vehicle (Utilities|Casting|CastToFortAthenaVehicle _array_element))
                (Default|KnockbackVehicle _asfort_athena_vehicle))))
          (Ability|ApplyGameplayEffectContainer _targetdata _applicationtag))))))

(event Ability|EventOnEndAbility
  (params (bWasCancelled :type Boolean))
  (|CallParentFunction bWasCancelled) ; This node is disabled and will not be called.\nDrag off pins to build functionality.
)

(fn KnockbackPawn
  (params (FortPawn :type Fort%20Pawn%20Object%20Reference))
  (bind (_returnvalue_x _returnvalue_y) (Math|Vector|BreakVector (Math|Vector|Normalize (- (Transformation|GetActorLocation FortPawn) (Transformation|GetActorLocation (Variables|InternalData|GetAIPawn))))))
  (Character|LaunchCharacter FortPawn (Math|Vector|MakeVector (* _returnvalue_x (ScalableFloat|GetValueatLevel (Variables|InternalData|GetKnockbackPawnForwardVelocity))) (* _returnvalue_y (ScalableFloat|GetValueatLevel (Variables|InternalData|GetKnockbackPawnForwardVelocity))) (ScalableFloat|GetValueatLevel (Variables|InternalData|GetKnockbackPawnHeightVelocity))) true true))

(fn KnockbackVehicle
  (params (FortAthenaVehicle :type Fort%20Athena%20Vehicle%20Object%20Reference))
  (bind (_returnvalue_x _returnvalue_y) (Math|Vector|BreakVector (Math|Vector|Normalize (- (Transformation|GetActorLocation FortAthenaVehicle) (Transformation|GetActorLocation (Variables|InternalData|GetAIPawn))))))
  (Audio|BroadcastAbilityImpactatLocation 0 FortAthenaVehicle (Math|Vector|MakeVector (* _returnvalue_x (ScalableFloat|GetValueatLevel (Variables|InternalData|GetKnockbackVehicleForwardVelocity))) (* _returnvalue_y (ScalableFloat|GetValueatLevel (Variables|InternalData|GetKnockbackVehicleForwardVelocity))) (ScalableFloat|GetValueatLevel (Variables|InternalData|GetKnockbackVehicleHeightVelocity))) 777.0 "None" false))
