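; Gameplay ability for the FirePetal wandering boss (Smash) that plays its Syphon idle montage at its patrol spot.
; Activation pauses target focusing, rotates the boss along the first patrol point's forward direction and plays the montage, with a forced-montage path alongside.
; An alert-level listener jumps to the montage's Outro or ends the ability unless the boss is DBNO.
; Ending adds a lock tag, resumes focusing and, if the montage played, applies attack-prevention effects.
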(event Ability|EventActivateAbility
  (bind _aicontroller (Variables|InternalData|GetAIController))
  (bind _returnvalue (Actor|GetComponentbyClass _aicontroller "/Script/FortniteAI.FortAthenaNpcPatrollingComponent"))
  (bind _returnvalue_2 (Class|FortAthenaNpcPatrollingComponent|GetPatrolPath _returnvalue :!class /Script/FortniteAI.FortAthenaNpcPatrollingComponent))
  (bind _returnvalue_3 (Class|FortAthenaPatrolPath|GetPatrolPoint _returnvalue_2 0))
  (bind _returnvalue_4 (Transformation|GetActorLocation _returnvalue_3))
  (|CallParentFunction)
  (Class|FortAthenaAIBotController|PauseTargetFocusing (Variables|InternalData|GetAIController))
  (Ability|CommitAbility)
  (AlertLevel|BindEventtoOnAlertLevelChangedEvent (Variables|InternalData|GetAIController) (delegate Custom|OnAlertLevelChanged))
  (CallFunction|OnAlertLevelChanged (Variables|InternalData|GetAIController) "Unaware" (Variables|/Script/FortniteAI.FortAthenaAIBotController|GetCurrentAlertLevel (Variables|InternalData|GetAIController)))
  (Variables|Default|SetbIdleMontagePlaying false)
  (Utilities|IsValid _returnvalue
    (:Is%20Valid
      (Utilities|IsValid _returnvalue_2
        (:Is%20Valid
          (Utilities|IsValid _returnvalue_3
            (:Is%20Valid
              (Variables|Default|SetAbilityLocation _returnvalue_4)
              (Variables|Default|SetAbilityFacing (+ _returnvalue_4 (* (Transformation|GetActorForwardVector _returnvalue_3) 5e+02)))
              (if (Variables|Debug|GetbDebugAbility)
                (Rendering|Debug|DrawDebugSphere (Variables|Default|GetAbilityFacing) 1e+02 12 "(R=0.805766,G=0.024942,B=1.000000,A=1.000000)" 1e+01))
              (bind _asynctaskproxy (AI|Tasks|RotateAIToFace "RotateAIToFace" 0 (Variables|Default|GetAbilityFacing) 5.0))
              (Utilities|IsValid _asynctaskproxy
                (:Is%20Valid
                  (bind _output_get (Variables|Default|SetRotateAsyncTask _asynctaskproxy))
                  (EventDispatchers|BindEventtoOnFinished _output_get (delegate Custom|OnRotateFinished)))
                (:Is%20Not%20Valid
                  (Ability|EndAbility))))
            (:Is%20Not%20Valid
              _)))
        (:Is%20Not%20Valid
          _))
      (CallFunction|ForceIdleMontage))
    (:Is%20Not%20Valid
      _)))

(event Custom|OnRotateFinished
  (bind (_montageinfo_montagetoplay _montageinfo_animplayrate _montageinfo_animrootmotiontranslationscale) (Utilities|Struct|BreakFortGameplayAbilityMontageInfo (Variables|GetMontageInfo_MontageToPlay)))
  (Ability|Tasks|PlayMontageandWaitforTrigger "WanderingBoss_SyphonIdle" _montageinfo_montagetoplay _montageinfo_animplayrate "FirstSection" "None" _montageinfo_animrootmotiontranslationscale
    (:then
      (Variables|Default|SetbIdleMontagePlaying true))
    (:Cancelled
      _)
    (:Completed
      _))
  (Ability|EndAbility))

(event Custom|OnAlertLevelChanged
  (params (BotController :type Fort%20Athena%20AIBot%20Controller%20Object%20Reference) (OldAlertLevel :type EAlertLevel%20Enum :default "Unaware") (NewAlertLevel :type EAlertLevel%20Enum :default "Unaware"))
  (if (not (Variables|/Script/FortniteGame.FortPawn|GetbIsDBNO (Variables|InternalData|GetAIPawn)))
    (switch Utilities|FlowControl|Switch|SwitchonEAlertLevel NewAlertLevel
      (:Suspicious
        _)
      (:LostTarget
        _)
      (:Threatened
        _))
    (if (Variables|Default|GetbIdleMontagePlaying)
      (Ability|Animation|MontageJumptoSection "Outro")
      (else
        (Ability|EndAbility)))))

(event Custom|ForceIdleMontage
  (bind (_montageinfo_montagetoplay _montageinfo_animplayrate _montageinfo_animrootmotiontranslationscale) (Utilities|Struct|BreakFortGameplayAbilityMontageInfo (Variables|GetMontageInfo_MontageToPlay)))
  (Ability|Tasks|PlayMontageandWaitforTrigger "WanderingBoss_SyphonIdle" _montageinfo_montagetoplay _montageinfo_animplayrate "FirstSection" "None" _montageinfo_animrootmotiontranslationscale
    (:then
      (Variables|Default|SetbIdleMontagePlaying true))
    (:Cancelled
      _)
    (:Completed
      _))
  (Ability|EndAbility))

(event Ability|EventOnEndAbility
  (params (bWasCancelled :type Boolean))
  (bind _movetoasynctask (Variables|Default|GetMoveToAsyncTask))
  (bind _rotateasynctask (Variables|Default|GetRotateAsyncTask))
  (bind _aicontroller (Variables|InternalData|GetAIController))
  (Ability|GameplayEffect|AddLooseGameplayTags (Variables|InternalData|GetAIPawn) (Variables|Default|GetSyphon%20Idle%20Lock%20Tag))
  (if (Variables|Default|GetbIdleMontagePlaying)
    (Ability|ApplyGameplayEffectToOwner "/FirePetal_WanderingBoss/Abilities/Smash/GE_Bombardment_Prevent_FirePetal_WanderingBoss_Smash.GE_Bombardment_Prevent_FirePetal_WanderingBoss_Smash_C")
    (Ability|ApplyGameplayEffectToOwner "/FirePetal_WanderingBoss/Abilities/Smash/GE_ProjectileFan_Prevent_FirePetal_WanderingBoss_Smash.GE_ProjectileFan_Prevent_FirePetal_WanderingBoss_Smash_C")
    (Ability|ApplyGameplayEffectToOwner "/FirePetal_WanderingBoss/Abilities/Smash/GE_SwordCombo_Prevent_FirePetal_WanderingBoss_Smash.GE_SwordCombo_Prevent_FirePetal_WanderingBoss_Smash_C")
    (Ability|ApplyGameplayEffectToOwner "/FirePetal_WanderingBoss/Abilities/Smash/GE_VoidWarp_Local_Prevent_FirePetal_WanderingBoss_Smash.GE_VoidWarp_Local_Prevent_FirePetal_WanderingBoss_Smash_C"))
  (AlertLevel|UnbindEventfromOnAlertLevelChangedEvent _aicontroller (delegate Custom|OnAlertLevelChanged))
  (Class|FortAthenaAIBotController|ResumeTargetFocusing _aicontroller)
  (Utilities|IsValid _movetoasynctask
    (:Is%20Valid
      (GameplayTasks|EndTask _movetoasynctask))
    (:Is%20Not%20Valid
      _))
  (Utilities|IsValid _rotateasynctask
    (:Is%20Valid
      (GameplayTasks|EndTask _rotateasynctask))
    (:Is%20Not%20Valid
      _))
  (|CallParentFunction bWasCancelled))
