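; Base Void Warp gameplay ability that teleports the FirePetal Wandering Boss (Smash) to an EQS-chosen location, facing its blackboard target.
; Activation takes AI behavior control, stops fire, applies target lock and runs the query.
; A montage notify triggers the trail cue, then a short timer teleports and calls OnExecutedTeleport, which only logs here.
; Ending, including on death, removes the lock, clears timers, restores AI control and applies local cooldown.
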(event Ability|EventActivateAbility
  (Development|PrintString "@@@ Event ActivateAbility [GA_VoidWarp_FirePetal_WanderingBoss]: CALLED!" false (Variables|Debug|GetbDebugAbility))
  (CallFunction|InitAbility)
  (bind _returnvalue (Ability|ApplyGameplayEffectToOwner (Variables|InternalData|GetGE_TargetLock)))
  (Variables|InternalData|SetTargetLockHandle _returnvalue)
  (bind _returnvalue_2 (Ability|CommitAbility))
  (if _returnvalue_2
    (bind _targetactor (Class|BPFLAICommonFunctions|GetTargetActorfromBlackboard (Variables|InternalData|GetFortPlayerPawn)))
    (Variables|InternalData|SetCurrentTargetActor _targetactor)
    (bind _returnvalue_3 (AI|EQS|RunEQSQuery (Variables|Tunables|GetWarpQuery) (Variables|InternalData|GetFortPlayerPawn)))
    (Utilities|IsValid _returnvalue_3
      (:Is%20Valid
        (AI|EQS|BindEventtoOnQueryFinished _returnvalue_3 (delegate Custom|OnWarpLocQueryFinished)))
      (:Is%20Not%20Valid
        (Development|PrintString "@@@@@@ Event ActivateAbility [GA_VoidWarp_FirePetal_WanderingBoss]:  ERROR - Player Query Instance INVALID!" false (Variables|Debug|GetbDebugAbility))
        (Ability|EndAbility)))
    (else
      (Ability|EndAbility))))

(event Custom|OnWarpLocQueryFinished
  (params (QueryInstance :type EQS%20Query%20Instance%20Object%20Reference) (QueryStatus :type EEnvQueryStatus%20Enum :default "Processing"))
  (bind (_montageinfo_montagetoplay _montageinfo_animplayrate _montageinfo_animrootmotiontranslationscale) (Utilities|Struct|BreakFortGameplayAbilityMontageInfo (Variables|GetMontageInfo_MontageToPlay)))
  (if (Utilities|Enum|Equal(Enum) QueryStatus "Success")
    (bind (_resultlocations _returnvalue) (AI|EQS|GetQueryResultsasLocations QueryInstance))
    (if _returnvalue
      (Variables|InternalData|SetTargetWarpLocation (Utilities|Array|Get(acopy) _resultlocations))
      (GameplayCue|ExecuteGameplayCue (Variables|InternalData|GetFortPlayerPawn) (Variables|GameplayTags|GetTeleportStartGameplayCueTag))
      (bind (_asynctaskproxy _notifytag) (Ability|Tasks|PlayMontageAndWaitNotify "WanderingBoss_VoidWarpAnim" _montageinfo_montagetoplay _montageinfo_animplayrate "None" true _montageinfo_animrootmotiontranslationscale)
        (:then
          (EventDispatchers|BindEventtoOnDied (Variables|InternalData|GetFortPlayerPawn) (EventDispatchers|CreateEvent :!event Custom|OnUserDeath)))
        (:OnNotifyBegin
          (switch Utilities|FlowControl|Switch|SwitchonGameplayTag _notifytag
            (:Event.WanderingBoss.VoidWarp.Fire
              (CallFunction|PlayTeleportTrail)))))
      (else
        (Development|PrintString "@@@@@@ Event ActivateAbility [GA_VoidWarp_FirePetal_WanderingBoss]:  ERROR - NO valid reults for Player!" false (Variables|Debug|GetbDebugAbility))
        (Ability|EndAbility)))
    (else
      (Development|PrintString "@@@@@@ Event ActivateAbility [GA_VoidWarp_FirePetal_WanderingBoss]: ERROR -Player Query Status NOT Success!" false (Variables|Debug|GetbDebugAbility))
      (Ability|EndAbility))))

(event Custom|PlayTeleportTrail
  (GameplayCue|ExecuteGameplayCue (Variables|InternalData|GetFortPlayerPawn) (Variables|GameplayTags|GetTeleportTrailGameplayCueTag))
  (bind _returnvalue (Utilities|Time|SetTimerbyEvent (EventDispatchers|CreateEvent :!event Custom|ExecuteTeleport) 0.1))
  (Variables|Timers|SetPlayTrailTeleportTimer _returnvalue))

(event Custom|ExecuteTeleport
  (GameplayCue|ExecuteGameplayCue (Variables|InternalData|GetFortPlayerPawn) (Variables|GameplayTags|GetTeleportExecutelGameplayCueTag))
  (bind _returnvalue (Transformation|Teleport (Variables|InternalData|GetFortPlayerPawn) (Utilities|Operators|Add (Variables|InternalData|GetTargetWarpLocation) (Math|Vector|MakeVector 0.0 0.0 (Components|Capsule|GetScaledCapsuleHalfHeight (Variables|/Script/Engine.Character|GetCapsuleComponent (Variables|InternalData|GetFortPlayerPawn))))) (Math|Rotator|FindLookatRotation (Variables|InternalData|GetTargetWarpLocation) (Pawn|GetNavAgentLocation (Variables|InternalData|GetCurrentTargetActor)))))
  (if _returnvalue
    (CallFunction|OnExecutedTeleport)
    (else
      (Development|PrintString "@@@@@@ Event ActivateAbility [GA_VoidWarp_FirePetal_WanderingBoss]: ERROR -Teleport NOT a Success!" false (Variables|Debug|GetbDebugAbility))))
  (Ability|EndAbility))

(event Ability|EventOnEndAbility
  (params (bWasCancelled :type Boolean))
  (bind _aibotcontroller (Variables|InternalData|GetAIBotController))
  (Development|PrintString "@@@ Event OnEndAbility [GA_VoidWarp_FirePetal_WanderingBoss]: CALLED!" false (Variables|Debug|GetbDebugAbility))
  (Ability|RemoveGameplayEffectFromOwnerWithHandle (Variables|InternalData|GetTargetLockHandle) 1)
  (Utilities|Time|ClearandInvalidateTimerbyHandle (Variables|Timers|GetExecuteTeleportTimer))
  (Utilities|Time|ClearandInvalidateTimerbyHandle (Variables|Timers|GetPlayTrailTeleportTimer))
  (EventDispatchers|UnbindEventfromOnDied (Variables|InternalData|GetFortPlayerPawn) (EventDispatchers|CreateEvent :!event Custom|OnUserDeath))
  (Utilities|IsValid _aibotcontroller
    (:Is%20Valid
      (AI|PopBehaviorControl _aibotcontroller)
      (Ability|ApplyGameplayEffectToOwner "/FirePetal_WanderingBoss/Abilities/Smash/GE_VoidWarp_Local_Cooldown_FirePetal_WanderingBoss_Smash.GE_VoidWarp_Local_Cooldown_FirePetal_WanderingBoss_Smash_C"))))

(event Custom|OnUserDeath
  (params (DamagedActor :type Actor%20Object%20Reference) (Damage :type Float%20%28single-precision%29) (InstigatedBy :type Controller%20Object%20Reference) (DamageCauser :type Actor%20Object%20Reference) (HitLocation :type Vector) (FHitComponent :type Primitive%20Component%20Object%20Reference) (BoneName :type Name) (Momentum :type Vector))
  (Ability|EndAbility))

(event Custom|OnExecutedTeleport
  (Development|PrintString "@@@ OnExecutedTeleport [GA_VoidWarp_FirePetal_WanderingBoss]: CALLED!" false (Variables|Debug|GetbDebugAbility)))

(fn InitAbility
  (bind _output_get (Variables|InternalData|SetFortPlayerPawn (CallFunction|GetAvatarasFortPlayerPawn)))
  (bind _asfort_athena_aibot_controller (Utilities|Casting|CastToFortAthenaAIBotController (Pawn|GetController _output_get))
    (:then
      (bind _output_get_2 (Variables|InternalData|SetAIBotController _asfort_athena_aibot_controller))
      (AI|PushBehaviorControl _output_get_2 -2097152)
      (Class|FortAthenaAIBotController|StopFire (Variables|InternalData|GetAIBotController)))
    (:CastFailed
      (Ability|EndAbility))))
