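; Ranged void-warp teleport ability for the FirePetal wandering boss's Smash variant, extending a parent ability's teleport events.
; Before teleporting, if the boss has the ammo threshold-crossed gameplay effect, it removes that effect and spawns ghost ammo loot above itself.
; After the parent's teleport completion logic, it rotates the boss toward its blackboard target using both nav agent locations, with an optional debug print.
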(event EventOnExecutedTeleport
  (bind _fortplayerpawn (Variables|InternalData|GetFortPlayerPawn))
  (|CallParentFunction)
  (Development|PrintString "@@@@@@ OnExecutedTeleport [GA_VoidWarp_Ranged_FirePetal_WanderingBoss]: CALLED!" false (Variables|Debug|GetbDebugAbility))
  (bind _targetactor (Class|BPFLAICommonFunctions|GetTargetActorfromBlackboard _fortplayerpawn))
  (Transformation|SetActorRotation (Variables|InternalData|GetFortPlayerPawn) (Math|Rotator|FindLookatRotation (Pawn|GetNavAgentLocation _fortplayerpawn) (Pawn|GetNavAgentLocation _targetactor)) false))

(event EventExecuteTeleport
  (bind _fortplayerpawn (Variables|InternalData|GetFortPlayerPawn))
  (bind _returnvalue (Ability|GetAbilitySystemComponent _fortplayerpawn))
  (if (> (GameplayEffects|GetGameplayEffectCount _returnvalue (Variables|Default|GetThresholdCrossed_Ammo) 0 true) 0)
    (GameplayEffects|RemoveActiveGameplayEffectbySourceEffect _returnvalue (Variables|Default|GetThresholdCrossed_Ammo) 0 -1)
    (Class|AthenaFunctionLibrary|SpawnLootfromTableUsingItemEntry "Loot_Pup_Ghost_Ammo" 1 -1 (+ (Pawn|GetNavAgentLocation (Variables|InternalData|GetFortPlayerPawn)) (* (Transformation|GetActorUpVector (Variables|InternalData|GetFortPlayerPawn)) 5e+02)) (Transformation|GetActorForwardVector (Variables|InternalData|GetFortPlayerPawn)) 0 true true false 0 "Other" "NPC"))
  (|CallParentFunction))
